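# HUD widget for an XP pop that shows the XP source label as a single slanted, uppercase yellow CommonTextBlock.
# SetSourceText formats the text from TextSource, appending an " xN" multiplier only when the multiple variable exceeds one.
# It defines an Intro animation that slides the text horizontally while fading it in, and a short bump scale punch.
# The graph itself plays neither animation.
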WidgetBlueprint /BattleRoyaleGameUI/HUD/XP/WBP_BRGameUI_HUD_XP_Pop_Source
animations [
 Intro {
  binding_count=1
  length_seconds=0.6074
  object=&self:Intro
  sequence {bindings [CommonTextBlock_ScoreType {
     id=#1
     possessed_class=CommonTextBlock
     tracks [
      {
       type=MovieScene2DTransformTrack
       sections [{
         type=MovieScene2DTransformSection
         channels [Translation.X {type=MovieSceneScriptingFloatChannel, keys [[0,86.1144],[3.0393,-1.061],[12.148,3.0957]]}]
        }]
      }
      {
       type=MovieSceneFloatTrack
       display_name=Render Opacity
       sections [{type=MovieSceneFloatSection, channels [None {type=MovieSceneScriptingFloatChannel, keys [[0,0],[2.163,1]]}]}]
      }
     ]
    }]}
 }
 bump {
  binding_count=1
  length_seconds=0.3235
  object=&self:bump
  sequence {bindings [CommonTextBlock_ScoreType {
     id=#2
     possessed_class=CommonTextBlock
     tracks [{
       type=MovieScene2DTransformTrack
       sections [{
         type=MovieScene2DTransformSection
         end_frame=6
         start_frame=0
         channels [
          Scale.X {type=MovieSceneScriptingFloatChannel, keys [[0,1],[1,1.3835],[6.4703,1]]}
          Scale.Y {type=MovieSceneScriptingFloatChannel, keys [[0,1],[1,1.2123],[6.4703,1]]}
         ]
        }]
      }]
    }]}
 }
]
graphs [EventGraph {nodes [
   EventSetSourceText() {is=title K2Node_CustomEvent}
   n1 = Get multiple() -> (multiple:int=0)
   n2 = "integer > integer"(A=@n1, B=1) {is=title K2Node_CallFunction}
   n3 = Format Text!(Format="LOCTEXT(\" x{multiplier}\")", multiplier:int=@n1) -> (Result)
   > "SetText (Text)"(
    self=Get CommonTextBlock_ScoreType() -> (CommonTextBlock_ScoreType:Common Text Object Reference)
    InText=Format Text!(
     Format="LOCTEXT(\"{Source}{multiplier}\")"
     Source:Text=Get TextSource() -> (TextSource:Text)
     multiplier:Text=Select(Option 0:Text, Option 1:Text=@n3, Index=@n2) -> (ReturnValue:Text)
    ) -> (Result)
   )
  ]}]
variables [
 multiple {type=integer, default=1}
 TextSource {type=string, default=""}
]
widgets {
 info {inheritedWidgetCount=0, parentClass=&UserWidget, rootWidgetClass=&CommonTextBlock, widgetCount=1}
 widgets [CommonTextBlock_ScoreType {
   widgetClassPath=&CommonTextBlock
   bIsVariable=true
   widget=CommonTextBlock_ScoreType
   properties {
    text=Source
    textTransformPolicy=ToUpper
    colorAndOpacity {colorUseRule=UseColor_Specified, specifiedColor=<rgba 1 0.9216 0.0782 1>}
    font {
     bForceMonospaced=false
     bMaterialIsStencil=false
     fontMaterial=null
     fontObject=&Burbank
     letterSpacing=0
     monospacedWidth=1
     size=15
     skewAmount=0
     typefaceFontName=Burbank-PlayerPower
     outlineSettings {
      bApplyOutlineToDropShadows=false
      bMiteredCorners=false
      bSeparateFillAlpha=false
      outlineMaterial=null
      outlineSize=1
      outlineColor=<rgba 0.8963 0.2307 0.0037 1>
     }
    }
    renderTransform {angle=0, scale=<1 1>, shear=<-10 0>, translation=<0 0>}
    renderTransformPivot=<0 0.5>
    shadowColorAndOpacity=<rgba 0.0052 0.0052 0.0056 0.55>
    shadowOffset=<4 3>
   }
  }]
}
