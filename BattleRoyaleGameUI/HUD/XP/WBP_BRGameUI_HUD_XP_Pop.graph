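# Battle Royale HUD XP pop widget that shows earned XP, or a simulated award size label, with rested-XP colors, icon and flare.
# Amount updates play Intro or ScoreBump, then a retriggerable delay starts Outro, which clears and hides the widget.
# Source updates stack child entries, merge consecutive repeats into a multiplier, and flush on an entry cap or timer.
# bUseSimpleAnim switches between curve-driven panel animations and UMG animations.
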WidgetBlueprint /BattleRoyaleGameUI/HUD/XP/WBP_BRGameUI_HUD_XP_Pop
animations [
 Intro {
  binding_count=3
  length_seconds=0.7711
  object=&self:Intro
  sequence {bindings [
    Overlay_TopContainer {
     id=#1
     possessed_class=Overlay
     tracks [{
       type=MovieSceneFloatTrack
       display_name=Render Opacity
       sections [{type=MovieSceneFloatSection, channels [None {type=MovieSceneScriptingFloatChannel, keys [[0.608,0],[2.8833,1],[6.5803,1]]}]}]
      }]
    }
    Image_Flare {
     id=#2
     possessed_class=Image
     tracks [
      {
       type=MovieSceneWidgetMaterialTrack
       sections [{
         type=MovieSceneParameterSection
         channels [TimeControl {type=MovieSceneScriptingFloatChannel, keys [[0.553,-8.6081],[3.805,-0.236],[13.284,0.3071]]}]
        }]
      }
      {
       type=MovieScene2DTransformTrack
       sections [{type=MovieScene2DTransformSection, channels [Scale.Y {type=MovieSceneScriptingFloatChannel, keys [[6.5847,1],[15.4223,0]]}]}]
      }
     ]
    }
    WidgetSwitcher_ScoreContainer {
     id=#3
     possessed_class=WidgetSwitcher
     tracks [{
       type=MovieScene2DTransformTrack
       sections [{
         type=MovieScene2DTransformSection
         channels [
          Scale.X {type=MovieSceneScriptingFloatChannel, keys [[0,0.0678],[1.984,1.6254],[3.573,1.0123],[4.627,0.9611],[5.0917,1.0389],[5.225,1]]}
          Scale.Y {type=MovieSceneScriptingFloatChannel, keys [[0,0.0678],[1.984,1.6254],[3.573,1.0123],[4.627,0.9611],[5.0917,1.0389],[5.225,1]]}
         ]
        }]
      }]
    }
   ]}
 }
 Outro {
  binding_count=1
  length_seconds=0.2678
  object=&self:Outro
  sequence {bindings [Overlay_TopContainer {
     id=#4
     possessed_class=Overlay
     tracks [
      {
       type=MovieSceneFloatTrack
       display_name=Render Opacity
       sections [{type=MovieSceneFloatSection, channels [None {type=MovieSceneScriptingFloatChannel, keys [[0,1],[5.3553,0]]}]}]
      }
      {
       type=MovieScene2DTransformTrack
       sections [{
         type=MovieScene2DTransformSection
         channels [
          Translation.X {type=MovieSceneScriptingFloatChannel, keys [[0,0],[5.3553,19.5671]]}
          Translation.Y {type=MovieSceneScriptingFloatChannel}
          Angle {type=MovieSceneScriptingFloatChannel}
          Scale.X {type=MovieSceneScriptingFloatChannel}
          Scale.Y {type=MovieSceneScriptingFloatChannel}
          Shear.X {type=MovieSceneScriptingFloatChannel}
          Shear.Y {type=MovieSceneScriptingFloatChannel}
         ]
        }]
      }
     ]
    }]}
 }
 ScoreBump {
  binding_count=1
  length_seconds=0.2833
  object=&self:ScoreBump
  sequence {
   bindings [Number_Score {
     id=#5
     possessed_class=CommonNumericTextBlock
     tracks [{
       type=MovieScene2DTransformTrack
       sections [{
         type=MovieScene2DTransformSection
         channels [
          Translation.X {type=MovieSceneScriptingFloatChannel}
          Translation.Y {type=MovieSceneScriptingFloatChannel}
          Angle {type=MovieSceneScriptingFloatChannel}
          Scale.X {type=MovieSceneScriptingFloatChannel, keys [[0,1],[1.631,1.7971],[10.382,0.9425],[17,1]]}
          Scale.Y {type=MovieSceneScriptingFloatChannel, keys [[0,1],[1.631,1.7971],[10.382,0.9425],[17,1]]}
          Shear.X {type=MovieSceneScriptingFloatChannel}
          Shear.Y {type=MovieSceneScriptingFloatChannel}
         ]
        }]
      }]
    }]
   display_rate {numerator=60}
  }
 }
]
graphs [EventGraph {nodes [
   n1 = OutroFinished() {is=title K2Node_CustomEvent}
   > UnbindFromOutroFinished()
   > EventEmptySources()
   > SetVisibility(InVisibility=Hidden)
   > ForwardEvents(self:Xp Everywhere Reticle Widget Object Reference)
   SetRestedVisuals() {is=title K2Node_CustomEvent}
   n2 = Get isRested() -> (isRested:bool)
   n3 = Select(
    Option 0:Linear Color Structure="Get (a copy)"(Array:Array of Linear Color Structures=Get NotRestedColor() -> (NotRestedColor:Array of Linear Color Structures), Dimension 1=1) -> (Output:Linear Color Structure)
    Option 1:Linear Color Structure="Get (a copy)"(Array=Get RestedColor() -> (RestedColor:Array of Linear Color Structures), Dimension 1=1)
    Index=@n2
   ) -> (ReturnValue:Linear Color Structure)
   > Set Outline Color!(
    In:Exec
    CommonTextWidget:Common Text Object Reference=Get Number_Score() -> (Number_Score:Common Numeric Text Block Object Reference)
    Color:Linear Color Structure=@n3
   ) -> (Out:Exec) {is=title K2Node_MacroInstance}
   > Set Outline Color(CommonTextWidget=Get CommonTextBlock_xp_plus() -> (CommonTextBlock_xp_plus:Common Text Object Reference), Color=@n3)
   > Set Outline Color(CommonTextWidget=Get CommonTextBlock_xp() -> (CommonTextBlock_xp:Common Text Object Reference), Color=@n3)
   > Set Outline Color(CommonTextWidget=Get CommonTextBlock_SimulatedXpSize() -> (CommonTextBlock_SimulatedXpSize:Common Text Object Reference), Color=@n3)
   > SetVisibility(
    self=Get RestedIcon() -> (RestedIcon:Rested Icon Object Reference)
    InVisibility=Select(Option 0:ESlateVisibility Enum=Collapsed, Option 1:ESlateVisibility Enum=HitTestInvisible, Index=@n2) -> (ReturnValue:ESlateVisibility Enum)
   )
   > SetBrushFromMaterial(
    self=Get Image_Flare() -> (Image_Flare:Image Object Reference)
    Material=Select(
     Option 0:Material Interface Object Reference=/Game/Items/Traps/Blueprints/Toys/UI_Art_Assets/M_UI_AccoladeFlare_Inst_nonRested
     Option 1:Material Interface Object Reference=/Game/Items/Traps/Blueprints/Toys/UI_Art_Assets/M_UI_AccoladeFlare_Inst
     Index=@n2
    ) -> (ReturnValue:Material Interface Object Reference)
   )
   EventEmptySources() {is=title K2Node_CustomEvent}
   > ClearChildren(self=Get VerticalBox_SourceContainer() -> (VerticalBox_SourceContainer:Vertical Box Object Reference))
   n4 = SetSourceRestedVisuals!() -> (OutputDelegate, then, SourceWidget:WBP BRGame UI HUD XP Pop Source Object Reference) {is=title K2Node_CustomEvent}
   n5 = Get CommonTextBlock_ScoreType!(self:WBP BRGame UI HUD XP Pop Source Object Reference=@n4.SourceWidget) -> (CommonTextBlock_ScoreType:Common Text Object Reference)
   > Set Outline Color(CommonTextWidget=@n5, Color=@n3)
   n6 = "Get (a copy)"(Array=Get NotRestedColor())
   n7 = "Get (a copy)"(Array=Get RestedColor())
   n8 = Get isRested()
   > SetColorAndOpacity(
    self=@n5
    InColorAndOpacity=Make Slate Color(SpecifiedColor=Select(Option 0:Linear Color Structure=@n6, Option 1:Linear Color Structure=@n7, Index=@n8) -> (ReturnValue:Linear Color Structure))
   )
   BindToOutroFinished() {is=title K2Node_CustomEvent}
   > n9 = Branch(Condition=Get bUseSimpleAnim() -> (bUseSimpleAnim:bool))
   n10 = OnAnimationsFinished!() -> (OutputDelegate, then, Name:Name) {is=title K2Node_CustomEvent}
   Bind Event to On Animations Finished(
    execute=@n9.then
    self:Fort Simple Widget Animations Panel Object Reference=Get Panel_Animations() -> (Panel_Animations:Fort Simple Widget Animations Panel Object Reference)
    Delegate=@n10
   ) {is=title K2Node_AddDelegate}
   BindToAnimationFinished(execute=@n9.else, Animation=Get Outro() -> (Outro:Widget Animation Object Reference), Delegate=@n1)
   UnbindFromOutroFinished() {is=title K2Node_CustomEvent}
   > n11 = Branch(Condition=Get bUseSimpleAnim())
   > n12 = Unbind all Events from On Animations Finished!(execute:Exec, self:Fort Simple Widget Animations Panel Object Reference=Get Panel_Animations()) -> (then:Exec) {is=title K2Node_ClearDelegate}
   Clear and Invalidate Timer by Handle(execute=[@n13.then, @n12.then], Handle=Get ForceFlushTimerHandle() -> (ForceFlushTimerHandle:Timer Handle Structure))
   n13 = UnbindAllFromAnimationFinished!(execute=@n11.else, self:User Widget Object Reference, Animation:Widget Animation Object Reference=Get Outro()) -> (then)
   n14 = Event UpdateXPSource!() -> (OutputDelegate, then, NewSourceText:"Text (by ref)") {is=title K2Node_Event}
   n15 = Get VerticalBox_SourceContainer()
   > n16 = Branch(Condition="integer >= integer"(
     A=GetChildrenCount!(self:Panel Widget Object Reference=@n15) -> (ReturnValue:int=0)
     B=Get MaxDisplayedEntries() -> (MaxDisplayedEntries:int=0)
    ))
   > n17 = OutroFinished() {is=title K2Node_CallFunction}
   n18 = Get VerticalBox_SourceContainer()
   n19 = Branch(execute=[@n17.then, @n16.else], Condition="Equal (Integer)"(A=GetChildrenCount(self=@n18) -> (ReturnValue=0)))
   > n20 = Set Timer by Event(Delegate=@n1, Time=Get MaxDisplayTime() -> (MaxDisplayTime:double), bLooping=false)
   > n21 = Set ForceFlushTimerHandle(ForceFlushTimerHandle:Timer Handle Structure=@n20) -> (Output_Get:Timer Handle Structure)
   n22 = Branch(execute=[@n21.then, @n19.else], Condition=TextIsEmpty(InText=@n14.NewSourceText))
   n23 = Length(TargetArray:Array of Widget Object References=GetAllChildren(self=@n18))
   n24 = Cast To WBP_BRGameUI_HUD_XP_Pop_Source!(
    execute=@n22.else
    Object=GetChildAt!(self:Panel Widget Object Reference=@n18, Index:int=int - int(A=@n23, B=1) {is=title K2Node_CallFunction}) -> (ReturnValue:Widget Object Reference)
   ) -> (then, CastFailed, AsWBP BRGame UI HUD XP Pop Source:WBP BRGame UI HUD XP Pop Source Object Reference) {is=title K2Node_DynamicCast}
   > n25 = Branch(Condition="Equal Exactly (Text)"!(
     A:"Text (by ref)"=@n14.NewSourceText
     B:"Text (by ref)"=Get TextSource!(self:WBP BRGame UI HUD XP Pop Source Object Reference=@n24) -> (TextSource:Text)
    ) -> (ReturnValue:bool))
   > Increment Int(Value=Get multiple!(self:WBP BRGame UI HUD XP Pop Source Object Reference=@n24) -> (multiple:int=0))
   > EventSetSourceText(self:WBP BRGame UI HUD XP Pop Source Object Reference=@n24)
   n26 = Create Widget!(
    execute:Exec=[@n25.else, @n24.CastFailed]
    Class:User Widget Class Reference=&/WBP_BRGameUI_HUD_XP_Pop_Source.WBP_BRGameUI_HUD_XP_Pop_Source_C
    OwningPlayer:Player Controller Object Reference
   ) -> (then:Exec, ReturnValue:WBP BRGame UI HUD XP Pop Source Object Reference)
   > AddChild(self=Get VerticalBox_SourceContainer(), Content=@n26)
   > Set TextSource!(execute, TextSource:Text=@n14.NewSourceText, self:WBP BRGame UI HUD XP Pop Source Object Reference=@n26) -> (then, Output_Get:Text)
   > EventSetSourceText(self=@n26)
   > SetSourceRestedVisuals!(execute, self, SourceWidget:WBP BRGame UI HUD XP Pop Source Object Reference=@n26) -> (then) {is=title K2Node_CallFunction}
   > PlayAnimation(self=@n26, InAnimation=Get Intro!(self:WBP BRGame UI HUD XP Pop Source Object Reference=@n26) -> (Intro))
   Event Construct()
   > EventEmptySources() {is=title K2Node_CallFunction}
   K2Node_Event_10 = UpdateEarnAmount!() -> (OutputDelegate, then, NewAmount:int, Cue:Sound Cue Object Reference, SimulatedXpSize:EFortSimulatedXPSize Enum=None, Denomination:Text) {is=title K2Node_CustomEvent}
   > Set Text(Text=@K2Node_Event_10.Denomination, self=Get CommonTextBlock_xp())
   > n27 = Sequence()
   n28 = Sequence(execute=@n27.then_1)
   > UnbindFromOutroFinished() {is=title K2Node_CallFunction}
   > n29 = Branch(Condition=Get bUseSimpleAnim())
   > n30 = StopAnimationByName!(
    execute
    self:Fort Simple Widget Animations Panel Object Reference=Get Panel_Animations()
    Name:"Name (by ref)"=Get OutroAnimName() -> (OutroAnimName:Name=None)
   ) -> (then, ReturnValue:bool)
   n31 = Branch(execute=[@n39.then, @n30.then], Condition=IsVisible())
   > n32 = Branch(Condition=Get bUseSimpleAnim())
   > n33 = PlayAnimationByName!(
    execute
    self:Fort Simple Widget Animations Panel Object Reference=Get Panel_Animations()
    Name:"Name (by ref)"=Get ScoreBumpAnimName() -> (ScoreBumpAnimName:Name=None)
    Speed:float=1
    StartAtTime:float=0
   ) -> (then, ReturnValue:bool)
   n34 = InterpolateToValue!(
    execute=[@n35.then, @n33.then]
    self:Common Numeric Text Block Object Reference=Get Number_Score()
    TargetValue:float="To Float (Integer)"(InInt=@K2Node_Event_10.NewAmount)
    MaximumInterpolationDuration:float=0.3
    MinimumChangeRate:float=0.1
    OutroOffset:float=0
   ) -> (then)
   PlaySound2D(execute=[@n38.then, @n37.then, @n34.then], Sound=@K2Node_Event_10.Cue)
   n35 = PlayAnimation(execute=@n32.else, InAnimation=Get ScoreBump() -> (ScoreBump:Widget Animation Object Reference))
   SetCurrentValue!(
    execute=@n31.else
    self:Common Numeric Text Block Object Reference=Get Number_Score()
    NewValue:float="To Float (Integer)"(InInt=@K2Node_Event_10.NewAmount)
   ) -> (then)
   > SetVisibility(InVisibility=HitTestInvisible)
   > DelayUntilNextTick!(execute) -> (then)
   > n36 = Branch(Condition=Get bUseSimpleAnim())
   > n37 = PlayAnimationByName(self=Get Panel_Animations(), Name=Get IntroAnimName() -> (IntroAnimName:Name=None), Speed=1, StartAtTime=0)
   n38 = PlayAnimation(execute=@n36.else, InAnimation=Get Intro!() -> (Intro))
   n39 = StopAnimation(execute=@n29.else, InAnimation=Get Outro())
   SetRestedVisuals(execute=@n28.then_1) {is=title K2Node_CallFunction}
   > RetriggerableDelay(Duration=Get OutroDelayTime() -> (OutroDelayTime:double))
   > BindToOutroFinished() {is=title K2Node_CallFunction}
   > n40 = Branch(Condition=Get bUseSimpleAnim())
   > PlayAnimationByName(self=Get Panel_Animations(), Name=Get OutroAnimName() -> (OutroAnimName=None), Speed=1, StartAtTime=0)
   PlayAnimation(execute=@n40.else, InAnimation=Get Outro())
   n41 = "Not Equal (Enum)"!(A:EFortSimulatedXPSize Enum=@K2Node_Event_10.SimulatedXpSize, B:EFortSimulatedXPSize Enum=None) -> (ReturnValue:bool)
   SetActiveWidgetIndex(
    execute=@n27.then_0
    self=Get WidgetSwitcher_ScoreContainer() -> (WidgetSwitcher_ScoreContainer:Widget Switcher Object Reference)
    Index=SelectInt(A=1, B=0, bPickA=@n41)
   )
   > Branch(Condition=@n41)
   > "SetText (Text)"(
    self=Get CommonTextBlock_SimulatedXpSize()
    InText=Format Text!(
     Format="LOCTEXT(\"<{0}>\")"
     0:Text=Select!(
      None:Text
      VerySmall:Text="LOCTEXT(\"VERY SMALL XP AWARD\")"
      Small:Text="LOCTEXT(\"SMALL XP AWARD\")"
      Medium:Text="LOCTEXT(\"MEDIUM XP AWARD\")"
      Large:Text="LOCTEXT(\"LARGE XP AWARD\")"
      VeryLarge:Text="LOCTEXT(\"VERY LARGE XP AWARD\")"
      Index:EFortSimulatedXPSize Enum=@K2Node_Event_10.SimulatedXpSize
     ) -> (ReturnValue:Text)
    ) -> (Result)
   )
   n42 = Event UpdateCustomDenominationEvent!() -> (
    OutputDelegate
    then
    Amount:int=0
    DenominationText:"Text (by ref)"
    Cue:Sound Cue Object Reference
    SimulatedXpSize:EFortSimulatedXPSize Enum=None
   ) {is=title K2Node_Event}
   > UpdateEarnAmount!(
    execute
    self
    NewAmount:int=@n42.Amount
    Cue:Sound Cue Object Reference=@n42.Cue
    SimulatedXpSize:EFortSimulatedXPSize Enum=@n42.SimulatedXpSize
    Denomination:Text=@n42.DenominationText
   ) -> (then) {is=title K2Node_CallFunction}
   n43 = Event UpdateRestXP!() -> (OutputDelegate, then, RestXpRemaining:int=0) {is=title K2Node_Event}
   > Set isRested(isRested:bool="integer > integer"(A=@n43.RestXpRemaining, B=0) {is=title K2Node_CallFunction}) -> (Output_Get:bool)
   n44 = Event UpdateXPAmount!() -> (OutputDelegate, then, NewAmount:int=0, Cue:Sound Cue Object Reference, SimulatedXpSize:EFortSimulatedXPSize Enum=None) {is=title K2Node_Event}
   > UpdateEarnAmount(NewAmount=@n44.NewAmount, Cue=@n44.Cue, SimulatedXpSize=@n44.SimulatedXpSize, Denomination="LOCTEXT(\"XP\")")
   OutroFinished(execute=@n45.Outro)
   n45 = Switch on Name!(execute:Exec=@n10.then, Selection:Name=@n10.Name) -> (Default:Exec, Outro:Exec) {is=title K2Node_SwitchName, pinNames [Outro]}
  ]}]
variables [
 bUseSimpleAnim {type=boolean, default=true}
 ForceFlushTimerHandle {type=TimerHandle}
 IntroAnimName {type=string, default=Intro}
 isRested {type=boolean, default=true}
 MaxDisplayedEntries {type=integer, default=10}
 MaxDisplayTime {type=number, default=10}
 NotRestedColor {type="array<LinearColor>", default [<rgba 1 1 1 1>, <rgba 0.1687 0 1 1>]}
 OutroAnimName {type=string, default=Outro}
 OutroDelayTime {type=number, default=2}
 RestedColor {type="array<LinearColor>", default [<rgba 1 0.9216 0.0782 1>, <rgba 1 0.3325 0 1>]}
 ScoreBumpAnimName {type=string, default=ScoreBump}
]
widgets {
 info {inheritedWidgetCount=1, parentClass=/Script/XpEverywhereUI.XpEverywhereReticleWidget, rootWidgetClass=&SizeBox, widgetCount=27}
 widgets [
  SizeBox_2 {widgetClassPath=&SizeBox, widget=SizeBox_2, properties {minDesiredHeight=400, minDesiredWidth=400}}
  Panel_Animations {
   widgetClassPath=/Script/FortniteUI.FortSimpleWidgetAnimationsPanel
   bIsVariable=true
   parent=SizeBox_2
   widget=Panel_Animations
   properties {animations [
     ScoreBump {
      bIsEnabled=true
      onAnimationsFinished=()
      targets [{
        alpha=null
        bResetOnFinish=true
        bShouldAnimateAlpha=false
        bShouldAnimateScale=true
        bShouldAnimateTranslation=false
        scale=/XpEverywhereUI/Curve_HUDXPPop_ScoreBump_NumberScore_Scale
        targetName=Number_Score
        translation=null
        materialScalarAnimations []
       }]
     }
     Intro {
      bIsEnabled=true
      onAnimationsFinished=()
      targets [
       {
        alpha=null
        bResetOnFinish=false
        bShouldAnimateAlpha=false
        bShouldAnimateScale=true
        bShouldAnimateTranslation=false
        scale=/XpEverywhereUI/Curve_HUDXPPop_Intro_ScoreContainer_Scale
        targetName=HorizontalBox_ScoreContainer
        translation=null
        materialScalarAnimations []
       }
       {
        alpha=null
        bResetOnFinish=false
        bShouldAnimateAlpha=false
        bShouldAnimateScale=true
        bShouldAnimateTranslation=false
        scale=/XpEverywhereUI/Curve_HUDXPPop_Intro_ImageFlare_Scale
        targetName=Image_Flare
        translation=null
        materialScalarAnimations [{materialParameter=TimeControl, scalarCurve=/XpEverywhereUI/Curve_HUDXPPop_Intro_ImageFlare_Material}]
       }
       {
        alpha=/XpEverywhereUI/Curve_HUDXPPop_Intro_TopContainer_Alpha
        bResetOnFinish=false
        bShouldAnimateAlpha=true
        bShouldAnimateScale=false
        bShouldAnimateTranslation=true
        scale=null
        targetName=Overlay_TopContainer
        translation=/XpEverywhereUI/Curve_HUDXPPop_Intro_TopContainer_Translation
        materialScalarAnimations []
       }
      ]
     }
     Outro {
      bIsEnabled=true
      onAnimationsFinished=()
      targets [{
        alpha=/XpEverywhereUI/Curve_HUDXPPop_Outro_TopContainer_Alpha
        bResetOnFinish=false
        bShouldAnimateAlpha=true
        bShouldAnimateScale=false
        bShouldAnimateTranslation=true
        scale=null
        targetName=Overlay_TopContainer
        translation=/XpEverywhereUI/Curve_HUDXPPop_Outro_TopContainer_Translation
        materialScalarAnimations []
       }]
     }
    ]}
   slot {class=SizeBoxSlot, properties {horizontalAlignment=HAlign_Left, verticalAlignment=VAlign_Top}}
  }
  Overlay_TopContainer {widgetClassPath=&Overlay, parent=Panel_Animations, widget=Overlay_TopContainer, slot {class=FortSimpleWidgetAnimationsSlot}}
  Image_shadow {
   widgetClassPath=&Image
   bIsVariable=true
   parent=Overlay_TopContainer
   widget=Image_shadow
   properties {
    brush {
     bIsDynamicallyLoaded=false
     drawAs=Image
     imageType=NoImage
     mirroring=NoMirror
     resourceName=None
     resourceObject=/Game/Athena/HUD/LTM/Wax/Materials/M_UI_TextGlow_Inst
     tiling=NoTile
     imageSize=<279.4116 128.8524>
     margin=<ltrb 0 0 0 0>
     outlineSettings {
      bUseBrushTransparency=false
      roundingType=HalfHeightRadius
      width=0
      color {colorUseRule=UseColor_Specified, specifiedColor=<rgba 0 0 0 0>}
      cornerRadii=<0 0 0 0>
     }
     tintColor {colorUseRule=UseColor_Specified, specifiedColor=<rgba 1 1 1 1>}
     uVRegion {bIsValid=false, max=<0 0>, min=<0 0>}
    }
    colorAndOpacity=<rgba 0.0052 0.0052 0.0056 0.18>
    renderTransform {angle=0, scale=<1 1>, shear=<0 0>, translation=<0 -5>}
   }
   slot {class=OverlaySlot, properties {horizontalAlignment=HAlign_Fill, verticalAlignment=VAlign_Fill, padding=<ltrb -125 -25 -125 -25>}}
  }
  VerticalBox_1 {
   widgetClassPath=&VerticalBox
   parent=Overlay_TopContainer
   widget=VerticalBox_1
   properties {visibility=HitTestInvisible}
   slot {class=OverlaySlot}
  }
  Overlay_0 {
   widgetClassPath=&Overlay
   parent=VerticalBox_1
   widget=Overlay_0
   slot {class=VerticalBoxSlot, properties {horizontalAlignment=HAlign_Left, verticalAlignment=VAlign_Top}}
  }
  Image_Flare {
   widgetClassPath=&Image
   bIsVariable=true
   parent=Overlay_0
   widget=Image_Flare
   properties {brush {
     bIsDynamicallyLoaded=false
     drawAs=Image
     imageType=NoImage
     mirroring=NoMirror
     resourceName=None
     resourceObject=&M_UI_AccoladeFlare_Inst
     tiling=NoTile
     imageSize=<174.6399 147.0081>
     margin=<ltrb 0 0 0 0>
     outlineSettings {
      bUseBrushTransparency=false
      roundingType=HalfHeightRadius
      width=0
      color {colorUseRule=UseColor_Specified, specifiedColor=<rgba 0 0 0 0>}
      cornerRadii=<0 0 0 0>
     }
     tintColor {colorUseRule=UseColor_Specified, specifiedColor=<rgba 1 1 1 1>}
     uVRegion {bIsValid=false, max=<0 0>, min=<0 0>}
    }}
   slot {class=OverlaySlot, properties {horizontalAlignment=HAlign_Fill, verticalAlignment=VAlign_Center, padding=<ltrb -100 -250 -100 -250>}}
  }
  SizeBox_1 {
   widgetClassPath=&SizeBox
   parent=Overlay_0
   widget=SizeBox_1
   properties {bOverride_MinDesiredWidth=true, minDesiredWidth=103.5022}
   slot {class=OverlaySlot}
  }
  WidgetSwitcher_ScoreContainer {
   widgetClassPath=&WidgetSwitcher
   bIsVariable=true
   parent=SizeBox_1
   widget=WidgetSwitcher_ScoreContainer
   properties {renderTransform {angle=0, scale=<1 1>, shear=<-8 0>, translation=<0 0>}, renderTransformPivot=<0.2 0.75>}
   slot {class=SizeBoxSlot}
  }
  HorizontalBox_259 {widgetClassPath=&HorizontalBox, parent=WidgetSwitcher_ScoreContainer, widget=HorizontalBox_259, slot {class=WidgetSwitcherSlot}}
  ScaleBox_0 {
   widgetClassPath=&ScaleBox
   parent=HorizontalBox_259
   widget=ScaleBox_0
   properties {stretch=UserSpecified, userSpecifiedScale=0.5}
   slot {class=HorizontalBoxSlot, properties {verticalAlignment=VAlign_Center, padding=<ltrb 0 0 -4 -2>}}
  }
  HorizontalBox_0 {widgetClassPath=&HorizontalBox, parent=ScaleBox_0, widget=HorizontalBox_0, slot {class=ScaleBoxSlot}}
  CommonTextBlock_xp_plus {
   widgetClassPath=&CommonTextBlock
   bIsVariable=true
   parent=HorizontalBox_0
   widget=CommonTextBlock_xp_plus
   properties {
    style=/Game/Athena/HUD/TextStyles/TextStyle-HUD-XPValue.TextStyle-HUD-XPValue_C
    text=+
    textTransformPolicy=ToUpper
    font {
     bForceMonospaced=false
     bMaterialIsStencil=false
     fontMaterial=null
     fontObject=&NotoSans
     letterSpacing=0
     monospacedWidth=1
     size=50
     skewAmount=0
     typefaceFontName=NotoSans-Regular
     outlineSettings {
      bApplyOutlineToDropShadows=false
      bMiteredCorners=false
      bSeparateFillAlpha=true
      outlineMaterial=null
      outlineSize=2
      outlineColor=<rgba 1 0.3325 0 1>
     }
    }
    renderTransformPivot=<0.2 0.75>
    shadowColorAndOpacity=<rgba 0.0052 0.0052 0.0056 0.55>
    shadowOffset=<6 6>
   }
   slot {class=HorizontalBoxSlot}
  }
  Number_Score {
   widgetClassPath=/Script/CommonUI.CommonNumericTextBlock
   bIsVariable=true
   parent=HorizontalBox_0
   widget=Number_Score
   properties {
    currentNumericValue=36.2573
    justification=Right
    style=&TextStyle-HUD-XPValue_C
    text=Text Block
    font {
     bForceMonospaced=false
     bMaterialIsStencil=false
     fontMaterial=null
     fontObject=&NotoSans
     letterSpacing=0
     monospacedWidth=1
     size=50
     skewAmount=0
     typefaceFontName=NotoSans-Regular
     outlineSettings {
      bApplyOutlineToDropShadows=false
      bMiteredCorners=false
      bSeparateFillAlpha=true
      outlineMaterial=null
      outlineSize=2
      outlineColor=<rgba 1 0.3325 0 1>
     }
    }
    renderTransformPivot=<0.2 0.8>
    shadowColorAndOpacity=<rgba 0.0052 0.0052 0.0056 0.55>
    shadowOffset=<6 6>
   }
   slot {class=HorizontalBoxSlot, properties {horizontalAlignment=HAlign_Center, verticalAlignment=VAlign_Center}}
  }
  CommonTextBlock_xp {
   widgetClassPath=&CommonTextBlock
   bIsVariable=true
   parent=HorizontalBox_0
   widget=CommonTextBlock_xp
   properties {
    style=&TextStyle-HUD-XPValue_C
    text=xp
    textTransformPolicy=ToUpper
    font {
     bForceMonospaced=false
     bMaterialIsStencil=false
     fontMaterial=null
     fontObject=&NotoSans
     letterSpacing=0
     monospacedWidth=1
     size=50
     skewAmount=0
     typefaceFontName=NotoSans-Regular
     outlineSettings {
      bApplyOutlineToDropShadows=false
      bMiteredCorners=false
      bSeparateFillAlpha=true
      outlineMaterial=null
      outlineSize=2
      outlineColor=<rgba 1 0.3325 0 1>
     }
    }
    renderTransform {angle=0, scale=<0.75 0.75>, shear=<0 0>, translation=<0 0>}
    renderTransformPivot=<0 0.5>
    shadowColorAndOpacity=<rgba 0.0052 0.0052 0.0056 0.55>
    shadowOffset=<6 6>
   }
   slot {class=HorizontalBoxSlot, properties {verticalAlignment=VAlign_Center, padding=<ltrb -2 0 0 -2>}}
  }
  RestedIcon {
   widgetClassPath=/Game/Athena/HUD/XP/RestedIcon
   bIsVariable=true
   parent=HorizontalBox_259
   widget=RestedIcon
   properties {renderTransform {angle=0, scale=<0.7 0.7>, shear=<0 0>, translation=<0 0>}}
   slot {class=HorizontalBoxSlot}
  }
  VerticalBox_0 {
   widgetClassPath=&VerticalBox
   parent=WidgetSwitcher_ScoreContainer
   widget=VerticalBox_0
   slot {class=WidgetSwitcherSlot, properties {padding=<ltrb 50 0 50 -2>}}
  }
  ScaleBox_1 {
   widgetClassPath=&ScaleBox
   parent=VerticalBox_0
   widget=ScaleBox_1
   properties {stretch=UserSpecified, userSpecifiedScale=0.4}
   slot {class=VerticalBoxSlot, properties {padding=<ltrb 0 0 0 -8>}}
  }
  CommonTextBlock_SimulatedXpSize {
   widgetClassPath=&CommonTextBlock
   bIsVariable=true
   parent=ScaleBox_1
   widget=CommonTextBlock_SimulatedXpSize
   properties {
    style=&TextStyle-HUD-XPValue_C
    text="<VERY LARGE XP AWARD>"
    textTransformPolicy=ToUpper
    font {
     bForceMonospaced=false
     bMaterialIsStencil=false
     fontMaterial=null
     fontObject=&NotoSans
     letterSpacing=0
     monospacedWidth=1
     size=50
     skewAmount=0
     typefaceFontName=NotoSans-Regular
     outlineSettings {
      bApplyOutlineToDropShadows=false
      bMiteredCorners=false
      bSeparateFillAlpha=true
      outlineMaterial=null
      outlineSize=2
      outlineColor=<rgba 1 0.3325 0 1>
     }
    }
    renderTransformPivot=<0.2 0.75>
    shadowColorAndOpacity=<rgba 0.0052 0.0052 0.0056 0.55>
    shadowOffset=<6 6>
   }
   slot {class=ScaleBoxSlot}
  }
  ScaleBox_2 {
   widgetClassPath=&ScaleBox
   parent=VerticalBox_0
   widget=ScaleBox_2
   properties {stretch=UserSpecified, userSpecifiedScale=0.75}
   slot {class=VerticalBoxSlot, properties {horizontalAlignment=HAlign_Center, verticalAlignment=VAlign_Center}}
  }
  CommonTextBlock_TestMode {
   widgetClassPath=&CommonTextBlock
   parent=ScaleBox_2
   widget=CommonTextBlock_TestMode
   properties {
    style=/Game/Athena/HUD/TextStyles/TextStyle-HUD-XPDescriptor.TextStyle-HUD-XPDescriptor_C
    text="[Test Mode]"
    textTransformPolicy=ToUpper
    font {
     bForceMonospaced=false
     bMaterialIsStencil=false
     fontMaterial=null
     fontObject=&Burbank
     letterSpacing=0
     monospacedWidth=1
     size=20
     skewAmount=0
     typefaceFontName=Burbank-PlayerPower
     outlineSettings {
      bApplyOutlineToDropShadows=false
      bMiteredCorners=false
      bSeparateFillAlpha=true
      outlineMaterial=&M-FontShadow
      outlineSize=0
      outlineColor=<rgba 0.1094 0.0129 0 0.5>
     }
    }
    renderTransformPivot=<0.2 0.75>
    shadowColorAndOpacity=<rgba 0.0052 0.0052 0.0056 0.4>
    shadowOffset=<2 2>
   }
   slot {class=ScaleBoxSlot}
  }
  VerticalBox_SourceContainer {
   widgetClassPath=&VerticalBox
   bIsVariable=true
   parent=VerticalBox_1
   widget=VerticalBox_SourceContainer
   properties {renderTransform {angle=0, scale=<1 1>, shear=<0 0>, translation=<25 -8>}}
   slot {class=VerticalBoxSlot, properties {horizontalAlignment=HAlign_Right, verticalAlignment=VAlign_Top}}
  }
  HUD_XP_Pop_Source {
   widgetClassPath=/XpEverywhereUI/HUD_XP_Pop_Source
   bIsVariable=true
   parent=VerticalBox_SourceContainer
   widget=HUD_XP_Pop_Source
   slot {class=VerticalBoxSlot}
  }
  HUD_XP_Pop_Source_1 {
   widgetClassPath=&HUD_XP_Pop_Source
   bIsVariable=true
   parent=VerticalBox_SourceContainer
   widget=HUD_XP_Pop_Source_1
   slot {class=VerticalBoxSlot}
  }
  HUD_XP_Pop_Source_2 {
   widgetClassPath=&HUD_XP_Pop_Source
   bIsVariable=true
   parent=VerticalBox_SourceContainer
   widget=HUD_XP_Pop_Source_2
   slot {class=VerticalBoxSlot}
  }
  HUD_XP_Pop_Source_3 {
   widgetClassPath=&HUD_XP_Pop_Source
   bIsVariable=true
   parent=VerticalBox_SourceContainer
   widget=HUD_XP_Pop_Source_3
   slot {class=VerticalBoxSlot}
  }
  TutorialHighlight {widgetClassPath=&TutorialHighlight}
 ]
}
