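# Battle Royale HUD element that hosts left-side toast notifications for accolades, events, XP progress, new levels and level-up rewards.
# A fixed-height size box wraps a prioritized widget switcher with zero transition time that holds each toast widget.
# Priorities arbitrate between toasts, with event notifications, new-level and XP bar widgets queuing behind when contested.
# The event graph has only empty construct events, so the asset is layout only.
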WidgetBlueprint /BattleRoyaleGameUI/HUD/XP/WBP_BRGameUI_LeftToasts
graphs [EventGraph {nodes [
   Event PreConstruct()
   Event Construct()
  ]}]
variables []
widgets {
 info {inheritedWidgetCount=1, parentClass=&FortHUDElementWidget, rootWidgetClass=&SizeBox, widgetCount=9}
 widgets [
  SizeBox_0 {widgetClassPath=&SizeBox, widget=SizeBox_0, properties {bOverride_HeightOverride=true, heightOverride=300}}
  Overlay_0 {widgetClassPath=&Overlay, parent=SizeBox_0, widget=Overlay_0, slot {class=SizeBoxSlot}}
  FortPrioritizedWidgetSwitcher {
   widgetClassPath=/Script/FortniteUI.FortPrioritizedWidgetSwitcher
   bIsVariable=true
   parent=Overlay_0
   widget=FortPrioritizedWidgetSwitcher
   properties {transitionDuration=0}
   slot {class=OverlaySlot}
  }
  WBP_BRGameUI_Accolades {
   widgetClassPath=&/WBP_BRGameUI_Accolades
   bIsVariable=true
   parent=FortPrioritizedWidgetSwitcher
   widget=WBP_BRGameUI_Accolades
   properties {priority=1, renderTransform {angle=0, scale=<1 1>, shear=<0 0>, translation=<0 25>}}
   slot {class=WidgetSwitcherSlot, properties {horizontalAlignment=HAlign_Left, padding=<ltrb 0 20 0 0>}}
  }
  EventNotifications {
   widgetClassPath=/FortniteCoreUI/UI/InGame/EventNotifications/EventNotifications
   bIsVariable=true
   parent=FortPrioritizedWidgetSwitcher
   widget=EventNotifications
   properties {prioritizationData {priority=Priority_0, whenContesting=QueueBehind}}
   slot {class=WidgetSwitcherSlot}
  }
  NewLevelWidget {
   widgetClassPath=/XpEverywhereUI/NewLevelWidget
   bIsVariable=true
   parent=FortPrioritizedWidgetSwitcher
   widget=NewLevelWidget
   properties {prioritizationData {priority=Priority_1, whenContesting=QueueBehind}}
   slot {class=WidgetSwitcherSlot, properties {horizontalAlignment=HAlign_Left, verticalAlignment=VAlign_Top, padding=<ltrb 0 45 0 0>}}
  }
  XP_Everywhere_Bar {
   widgetClassPath=/XpEverywhereUI/XP_Everywhere_Bar
   bIsVariable=true
   parent=FortPrioritizedWidgetSwitcher
   widget=XP_Everywhere_Bar
   properties {prioritizationData {priority=Priority_1, whenContesting=QueueBehind}}
   slot {class=WidgetSwitcherSlot, properties {horizontalAlignment=HAlign_Left, verticalAlignment=VAlign_Center, padding=<ltrb -16 25 0 0>}}
  }
  LevelUpRewardsNotification {
   widgetClassPath=/XpEverywhereUI/LevelUpRewardsNotification
   bIsVariable=true
   parent=FortPrioritizedWidgetSwitcher
   widget=LevelUpRewardsNotification
   properties {priority=1}
   slot {class=WidgetSwitcherSlot}
  }
  TutorialHighlight {widgetClassPath=&TutorialHighlight}
 ]
}
